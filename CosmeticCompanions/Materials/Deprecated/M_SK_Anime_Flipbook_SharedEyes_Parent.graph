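# Deprecated cloth-model parent material for anime-shaded companion skeletal meshes with flipbook faces and animated eyes.
# The 2023 anime shader Pre, Calculations and Post functions form the base, with normals overridden by a companion normal-sphere function.
# Flipbook face color blends into lit and shaded colors before eye animation, masked by texture, vertex color or second-UV sign.
# The face function also adds World Position Offset, and a temporal responsiveness output is blended.
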Material /CosmeticCompanions/Materials/Deprecated/M_SK_Anime_Flipbook_SharedEyes_Parent
graph {
 properties {
  bAutomaticallySetUsageInEditor=false
  bUseMaterialAttributes=true
  bUsedWithClothing=true
  bUsedWithMeshDeformer=true
  bUsedWithMorphTargets=true
  bUsedWithNiagaraMeshParticles=true
  bUsedWithSkeletalMesh=true
  bUsedWithStaticLighting=true
  pixelDepthOffsetMode=PDOM_Legacy
  shadingModel=MSM_Cloth
  usedShadingModels=Cloth
 }
 nodes [
  n1 = MaterialFunctionCall!(CustomUV) {materialFunction=/Game/Global/Materials/Character_Parent_2023/Anime_Parent_2023/MaterialFunctions/MFC_AnimeShader_2023_Pre}
  n2 = MaterialFunctionCall!(
   CustomLightVector
   OverrideNormal_WS=MaterialFunctionCall!(Normal=@n1.Normal) {materialFunction=/CosmeticCompanions/Materials/MaterialFunctions/MF_CompanionAnime_NormalSphere}.Normal
   STT=@n1.STT
   Normal_Map=@n1.STT
   DFL=@n1.STT
   FX=@n1.STT
  ) {materialFunction=/Game/Global/Materials/Character_Parent_2023/Anime_Parent_2023/MaterialFunctions/MFC_AnimeShader_2023_Calculations}
  n3 = MaterialFunctionCall!(Material, Brow UV, Eye UV, Mouth UV) {materialFunction=/CosmeticCompanions/Materials/MaterialFunctions/MF_FlipbookFace_Companion}
  n4 = NamedRerouteDeclaration(None=@n3.FaceAlpha) {name=FlipbookFaceAlpha, nodeColor=<rgba 0.8281 0 1 1>}
  n5 = ScalarParameter() {defaultValue=1, parameterName=FlipbookFaceAnimeShadedValue}
  n6 = VertexColor()
  n7 = TextureSampleParameter2D() {
   group=01_EyeAnimation_Parameters
   parameterName=EyeMask
   samplerSource=SSM_Wrap_WorldGroupSettings
   samplerType=SAMPLERTYPE_Masks
   sortPriority=-32
   texture=&T_WhiteMask
  }
  n8 = StaticSwitchParameter(True=AppendVector(A=@n6, B=@n6), False=@n7.RGBA) {group=01_EyeAnimation_Parameters, parameterName=UseVertexColorForEyeMask, sortPriority=-33}
  n9 = TextureCoordinate() {coordinateIndex=1}
  n10 = Sign(None=Subtract(A=ComponentMask(None=@n9) {r=true}) {constB=0.0001})
  n11 = MaterialFunctionCall!(
   Color Lit=LinearInterpolate(A=@n1.Color_Lit, B=@n3.FaceColor, Alpha=@n4)
   Color Shaded=LinearInterpolate(A=@n1.Color_Shaded, B=Multiply(A=@n3.FaceColor, B=@n5), Alpha=@n4)
   Use UV2=StaticBoolParameter() {defaultValue=true, group=01_EyeAnimation_Parameters, parameterName=Use UV2 for Eye Animation, sortPriority=-30}
   Mask=StaticSwitchParameter(
    True=ChannelMaskParameter(None=@n8) {group=01_EyeAnimation_Parameters, parameterName=EyeMaskChannel, sortPriority=-32}.R
    False=Saturate(None=@n10)
   ) {group=01_EyeAnimation_Parameters, parameterName=UseEyeMask, sortPriority=-33}
  ) {materialFunction=/CosmeticCompanions/Materials/MaterialFunctions/MF_CompanionAnimeEyeAnimation}
  n12 = GetMaterialAttributes(None=MaterialFunctionCall!(
    Shading_Gradient=@n2.Shading_Gradient
    Threshold_Gradient=@n2.Shading_Gradient
    Specular_Gradient=@n2.Shading_Gradient
    Sheen_Mask=@n2.Shading_Gradient
    Thickness=@n2.Shading_Gradient
    DFL_Mask=@n2.Shading_Gradient
    FX=@n2.Shading_Gradient
    Metallic_Gradient=@n2.Shading_Gradient
    RimLight_Gradient=@n2.Shading_Gradient
    NormalBuffer=@n2.Shading_Gradient
    --------
    CINE_LightSources=@n2.Shading_Gradient
    CINE_LightColor=@n2.Shading_Gradient
    ---------
    Color_Lit=@n11.Lit
    Color_Shaded=@n11.Lit
    SRM=@n1.SRM
    Emissive=@n1.SRM
    DFL_Color=@n1.SRM
    Specular_Color=@n1.SRM
    RimLight_Color=@n1.SRM
   ) {materialFunction=/Game/Global/Materials/Character_Parent_2023/Anime_Parent_2023/MaterialFunctions/MFC_AnimeShader_2023_Post}.Result) {attributeGetTypes [#1]}
  SetMaterialAttributes!(
   MaterialAttributes=@n12.MaterialAttributes
   World Position Offset=Add(
    A=@n12.World Position Offset
    B=GetMaterialAttributes(None=@n3.Material) {attributeGetTypes [#1]}.World Position Offset
   )
  ) {attributeSetTypes [#1]}
  TemporalResponsivenessOutput!(None=LinearInterpolate(B=ScalarParameter() {parameterName=Temporal Responsiveness}, Alpha=NamedRerouteUsage()) {constA=0.4})
 ]
}
